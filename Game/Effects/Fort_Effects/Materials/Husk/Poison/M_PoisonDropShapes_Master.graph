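# Unlit translucent master material for poison droplet particle sprites on Husk effects, using packed 4x4 droplet textures.
# Fakes lighting by dotting the droplet normal map with the global sunlight vector, then adding shadow color and multiplying by Fresnel and particle color.
# Branches feed this into an emissive-adjustment function and a dynamic-parameter blend adding the texture's red channel.
# Particle alpha multiplies texture alpha, and a vertex-color jiggle function supplies world-position-offset wobble.
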Material /Game/Effects/Fort_Effects/Materials/Husk/Poison/M_PoisonDropShapes_Master
graph {
 properties {
  bEnableResponsiveAA=true
  bUsedWithNiagaraSprites=true
  bUsedWithParticleSprites=true
  blendMode=BLEND_Translucent
  pixelDepthOffsetMode=PDOM_Legacy
  shadingModel=MSM_Unlit
  usedShadingModels=Unlit
 }
 nodes [
  n1 = VectorParameter() {parameterName=ShadowColor, defaultValue=<rgba 0.1102 0.335 0.2096 1>}
  n2 = CollectionParameter() {
   collection=&FortniteMaterialParameters
   parameterName=SunlightDirectionalVector
  }
  n3 = ComponentMask(None=@n2) {b=true, g=true, r=true}
  n4 = TextureSampleParameter2D() {parameterName=Normal, samplerType=SAMPLERTYPE_Normal, texture=/Game/Effects/Fort_Effects/Textures/Liquid/T_PoisonDroplets_4X4_N}
  n5 = Multiply(A=@n4.RGB, B=VectorParameter() {parameterName=NormalStrength, defaultValue=<rgba 1 1 2 0>}.RGB)
  n6 = Power(Base=Clamp(None=DotProduct(A=@n3, B=@n5)), Exp=ScalarParameter() {defaultValue=2, parameterName=ShadowEdgePower})
  n7 = ScalarParameter() {defaultValue=48, parameterName=Fresnel_Exponent}
  n8 = ScalarParameter() {defaultValue=0.125, parameterName=Frensel_ReflectFrac}
  n9 = ParticleColor()
  n10 = Multiply(A=Multiply(A=Add(A=@n1.RGB, B=@n6), B=Fresnel(ExponentIn=@n7, BaseReflectFractionIn=@n8, Normal=@n4.RGB)), B=@n9.RGB)
  n11 = TextureSampleParameter2D() {parameterName=Color and Alpha, texture=/Game/Effects/Fort_Effects/Textures/Liquid/T_PoisonDropletsPacked_4X4}
  LinearInterpolate(
   A=@n10
   B=Add(A=@n10, B=@n11.R)
   Alpha=DynamicParameter() {defaultValue=<rgba 0 0 0 1>, paramNames [LerpAddColor_RedChannel, Param2, Param3, Param4]}.LerpAddColor_RedChannel
  )
  MaterialFunctionCall(Emissive=@n10) {materialFunction=&MF_FX_EmissiveAdjustments}
  MaterialFunctionCall!(
   Speed=ScalarParameter() {defaultValue=1, parameterName=WPO_Speed}
   LargeWaveScle=ScalarParameter() {defaultValue=1, parameterName=WPO_LargeWaveScale}
   WaveScale=ScalarParameter() {defaultValue=1.5, parameterName=WPO_WaveScale}
  ) {materialFunction=/Game/Effects/Fort_Effects/Materials/_Functions/Jiggle_WPO_VertColor_Red}
  Multiply(A=@n9.A, B=@n11.A)
 ]
}
