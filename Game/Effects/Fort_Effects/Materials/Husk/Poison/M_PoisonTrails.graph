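# Translucent unlit, two-sided material for Husk poison trails, usable on beam trails and particle sprites.
# Fakes sun lighting by dotting a panning tiled normal map, combined with the pixel normal, against the global sunlight vector.
# Inverted lighting, a shadow tint and trail-texture green are summed, clamped and multiplied by particle color for emissive.
# Trail and particle alpha, view-angle fade and Fade drive opacity, while dynamic parameters control texture tiling.
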Material /Game/Effects/Fort_Effects/Materials/Husk/Poison/M_PoisonTrails
graph {
 properties {
  bUsedWithBeamTrails=true
  bUsedWithParticleSprites=true
  blendMode=BLEND_Translucent
  pixelDepthOffsetMode=PDOM_Legacy
  shadingModel=MSM_Unlit
  twoSided=true
  usedShadingModels=Unlit
 }
 nodes [
  n1 = TextureCoordinate() {coordinateIndex=1}
  n2 = DynamicParameter() {defaultValue=<rgba 0 0 0 1>, paramNames [TileMask,TileNormal,Param3,Param4]}
  n3 = Time()
  n4 = ScalarParameter() {defaultValue=0.5, parameterName=PanSpeed}
  n5 = TextureSample(UVs=Panner(Coordinate=Multiply(A=@n1, B=@n2.TileNormal), Time=Multiply(A=@n3, B=@n4)) {speedY=-1}) {samplerType=SAMPLERTYPE_Normal, texture=/Game/Effects/Fort_Effects/Textures/Husk/Poison/T_PoisonTile_N}
  n6 = VectorParameter() {parameterName=NormalIntensity, defaultValue=<rgba 1 1 0.1 0>}
  n7 = CollectionParameter() {
   collection=&FortniteMaterialParameters
   parameterName=SunlightDirectionalVector
  }
  n8 = DotProduct(A=Multiply(A=Multiply(A=@n5.RGB, B=@n6.RGB), B=PixelNormalWS()), B=Multiply(A=ComponentMask(None=@n7) {b=true, g=true, r=true}) {constB=-1})
  n9 = Power(Base=Clamp(None=ConstantBiasScale(None=@n8)), Exp=ScalarParameter() {defaultValue=1, parameterName=LightingPower}) {constExponent=1}
  n10 = ComponentMask(None=TextureCoordinate() {coordinateIndex=1}) {r=true}
  n11 = DynamicParameter() {defaultValue=<rgba 0 0 0 1>, paramNames [TileMask,TileNormal,Param3,Param4]}
  n12 = TextureCoordinate() {coordinateIndex=1, uTiling=0}
  n13 = TextureSample(UVs=AppendVector(A=Multiply(A=@n10, B=@n11.TileMask), B=ComponentMask(None=@n12) {g=true})) {texture=/Game/Effects/Fort_Effects/Textures/Smoke/T_Lobber_Trail}
  n14 = Add(A=Add(A=VectorParameter() {parameterName=ShadowColor, defaultValue=<rgba 0.1105 1.5 0.4477 1>}.RGB, B=OneMinus(None=@n9)), B=@n13.G)
  n15 = ParticleColor()
  MaterialFunctionCall(Emissive=Multiply(A=Clamp(None=@n14), B=@n15.RGB)) {materialFunction=&MF_FX_EmissiveAdjustments}
  Multiply(
   A=Multiply(
    A=Multiply(A=@n15.A, B=@n13.A)
    B=MaterialFunctionCall() {materialFunction=&FadeBasedOnViewAngle}.Result
   )
   B=ScalarParameter() {defaultValue=1, parameterName=Fade}
  )
 ]
}
